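; Gameplay ability for the Exotic ColdFern weapon in WildEstate that, going by its name, responds to damage events.
; It reads instigator and target from the event data and checks that their team affiliation is Hostile.
; If the target also casts to FortPawn, it applies the Spicy gameplay effect to the ability owner, not the target.
; EndAbility appears only in the hostile branch; the non-hostile branch does nothing.
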(event Ability|EventActivateAbilityFromEvent
  (params (EventData :type Gameplay%20Event%20Data%20Structure%20%28by%20ref%29))
  (bind (_instigator _target) (Utilities|Struct|BreakGameplayEventData EventData))
  (if (Utilities|Enum|Equal(Enum) (Team|GetTeamAffiliation _instigator _target) "Hostile")
    (Utilities|Casting|CastToFortPawn _target
      (:then
        (Ability|ApplyGameplayEffectToOwner (Variables|Default|GetSpicyGameplayEffectClass)))
      (:CastFailed
        _))
    (Ability|EndAbility)
    (else
      _)))
