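; Gameplay ability for the exotic ColdFern weapon that applies a "spicy" gameplay effect to its owner when the owner takes damage.
; On activation it binds to the avatar player pawn's OnDamaged dispatcher.
; Each damage event applies the configured effect class only if the instigating controller resolves to a valid Fort pawn, regardless of damage amount.
; The graph never ends the ability or unbinds the delegate.
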(event Ability|EventActivateAbility
  (EventDispatchers|AssignOnDamaged (CallFunction|GetAvatarasFortPlayerPawn) (delegate Custom|OnDamaged_Event)))

(event Custom|OnDamaged_Event
  (params (DamagedActor :type Actor%20Object%20Reference) (Damage :type Float%20%28single-precision%29) (InstigatedBy :type Controller%20Object%20Reference) (DamageCauser :type Actor%20Object%20Reference) (HitLocation :type Vector) (FHitComponent :type Primitive%20Component%20Object%20Reference) (BoneName :type Name) (Momentum :type Vector))
  (Class|FortQuestBlueprintLibrary|GetFortPawnFromObject InstigatedBy
    (:Valid
      (Ability|ApplyGameplayEffectToOwner (Variables|Default|GetSpicyGameplayEffectClass)))))
